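# Looping ambient leaf-rustle Sound Cue for the hideable big leaf pile, assigned to the Fort_Traps sound class.
# A Random node picks one of three equally weighted looping waves, an Oscillator slowly swells its volume, and a root Modulator randomizes pitch and volume.
# Inline overrides take the place of the referenced attenuation asset, adding distance low-pass filtering and listener focus.
# Overridden concurrency allows only a few simultaneous instances, stopping the quietest.
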SoundCue /BigLeafPile/Environmental/Audio/Hideable_BigLeafPile_AmbientLoop_Cue
root=Modulator_1
nodes [
 {id=Modulator_1, children [Oscillator_0], properties {pitchMax=1.3, pitchMin=0.8, volumeMax=1, volumeMin=0.7}}
 {
  id=Oscillator_0
  children [Random_1]
  properties {amplitudeMax=0.4, amplitudeMin=0.4, bModulateVolume=true, centerMax=1, centerMin=1, frequencyMax=0.2, frequencyMin=0.2}
 }
 {id=Random_1, children [WavePlayer_7,WavePlayer_8,WavePlayer_9], properties {weights [1,1,1]}}
 {id=WavePlayer_7, properties {bLooping=true, soundWaveAssetPtr=&/BigLeafPile_Ambient_Leaves_Loop_01}}
 {id=WavePlayer_8, properties {bLooping=true, soundWaveAssetPtr=&/BigLeafPile_Ambient_Leaves_Loop_02}}
 {id=WavePlayer_9, properties {bLooping=true, soundWaveAssetPtr=&/BigLeafPile_Ambient_Leaves_Loop_03}}
]
properties {
 attenuationSettings=&/Hideable_BigLeafPile_Att
 bOverrideAttenuation=true
 bOverrideConcurrency=true
 maxDistance=2000
 soundClassObject=&Fort_Traps
 volumeMultiplier=1.25
 attenuationOverrides {
  bAttenuateWithLPF=true
  bEnableFocusInterpolation=true
  bEnableListenerFocus=true
  distanceAlgorithm=NaturalSound
  falloffDistance=1500
  hPFFrequencyAtMax=300
  lPFFrequencyAtMax=1200
  lPFRadiusMax=1500
  lPFRadiusMin=0
  nonFocusAzimuth=90
  nonFocusVolumeAttenuation=0.5
  nonSpatializedRadiusStart=300
  stereoSpread=500
  attenuationShapeExtents {x=500}
 }
 concurrencyOverrides {maxCount=3, resolutionRule=StopQuietest, retriggerTime=0.2, voiceStealReleaseTime=0.2}
}
