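# Sound cue for the destruction of the BigLeafPile hideable, routed to the Fort_Traps sound class with its own attenuation asset.
# A distance crossfade blends two random groups of three waves each: standard variants up close, mono-named variants farther away.
# A root modulator fixes a lowered pitch on the blended result, and a cue-level multiplier boosts overall volume.
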SoundCue /BigLeafPile/Environmental/Audio/Hideable_BigLeafPile_Destroyed_Cue
root=Modulator_0
nodes [
 {
  id=DistanceCrossFade_0
  children [Random_1,Random_2]
  properties {crossFadeInput [
    {fadeInDistanceEnd=0, fadeInDistanceStart=0, fadeOutDistanceEnd=1000, fadeOutDistanceStart=600, volume=1}
    {fadeInDistanceEnd=1000, fadeInDistanceStart=600, fadeOutDistanceEnd=0, fadeOutDistanceStart=0, volume=1}
   ]}
 }
 {id=Modulator_0, children [DistanceCrossFade_0], properties {pitchMax=0.8, pitchMin=0.8, volumeMax=1, volumeMin=1}}
 {id=Random_1, children [WavePlayer_3,WavePlayer_4,WavePlayer_5], properties {weights [1,1,1]}}
 {id=Random_2, children [WavePlayer_6,WavePlayer_7,WavePlayer_8], properties {weights [1,1,1]}}
 {id=WavePlayer_3, properties {soundWaveAssetPtr=&/BigLeafPile_Destroyed_01}}
 {id=WavePlayer_4, properties {soundWaveAssetPtr=&/BigLeafPile_Destroyed_02}}
 {id=WavePlayer_5, properties {soundWaveAssetPtr=&/BigLeafPile_Destroyed_03}}
 {id=WavePlayer_6, properties {soundWaveAssetPtr=&/BigLeafPile_Destroyed_mono_01}}
 {id=WavePlayer_7, properties {soundWaveAssetPtr=&/BigLeafPile_Destroyed_mono_02}}
 {id=WavePlayer_8, properties {soundWaveAssetPtr=&/BigLeafPile_Destroyed_mono_03}}
]
properties {
 attenuationSettings=&/Hideable_BigLeafPile_Att
 maxDistance=7800
 soundClassObject=&Fort_Traps
 volumeMultiplier=3
}
