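# Sound cue for entering the big leaf pile hideable, with spatial attenuation and the traps sound class.
# A mixer layers two branches: a random fast "exit" rustle with pitch and volume modulation, and a quieter random medium leaf-debris sound with volume variation.
# The main layer reuses Exit_Fast waves despite this being the Enter cue.
# Concurrency is limited per owner and virtualization is disabled.
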SoundCue /BigLeafPile/Environmental/Audio/Hideable_BigLeafPile_Enter_Cue
root=Mixer_0
nodes [
 {id=Mixer_0, children [Modulator_0,Modulator_1], properties {inputVolume [1,0.6]}}
 {id=Modulator_0, children [Random_0], properties {pitchMax=1.1, pitchMin=0.85, volumeMax=1.7, volumeMin=1}}
 {id=Modulator_1, children [Random_1], properties {volumeMin=0.7}}
 {id=Random_0, children [WavePlayer_0,WavePlayer_1,WavePlayer_2], properties {weights [1,1,1]}}
 {id=Random_1, children [WavePlayer_3,WavePlayer_4,WavePlayer_5,WavePlayer_6,WavePlayer_7,WavePlayer_8], properties {weights [1,1,1,1,1,1]}}
 {id=WavePlayer_0, properties {soundWaveAssetPtr=&/BigLeafPile_Exit_Fast_01}}
 {id=WavePlayer_1, properties {soundWaveAssetPtr=&/BigLeafPile_Exit_Fast_02}}
 {id=WavePlayer_2, properties {soundWaveAssetPtr=&/BigLeafPile_Exit_Fast_03}}
 {id=WavePlayer_3, properties {soundWaveAssetPtr=&/BigLeafPile_Leaf_Debris_Medium_01}}
 {id=WavePlayer_4, properties {soundWaveAssetPtr=&/BigLeafPile_Leaf_Debris_Medium_02}}
 {id=WavePlayer_5, properties {soundWaveAssetPtr=&/BigLeafPile_Leaf_Debris_Medium_03}}
 {id=WavePlayer_6, properties {soundWaveAssetPtr=&/BigLeafPile_Leaf_Debris_Medium_04}}
 {id=WavePlayer_7, properties {soundWaveAssetPtr=&/BigLeafPile_Leaf_Debris_Medium_05}}
 {id=WavePlayer_8, properties {soundWaveAssetPtr=&/BigLeafPile_Leaf_Debris_Medium_06}}
]
properties {
 attenuationSettings=&/Hideable_BigLeafPile_Att
 bOverrideConcurrency=true
 maxDistance=7800
 soundClassObject=&Fort_Traps
 virtualizationMode=Disabled
 volumeMultiplier=2
 concurrencyOverrides {bLimitToOwner=true, maxCount=2}
}
