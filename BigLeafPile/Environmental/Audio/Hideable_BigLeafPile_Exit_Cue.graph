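# Sound cue for exiting the BigLeafPile hideable.
# A mixer layers two branches, each randomly picking one of three equally weighted waves: a lower-pitched, quieter "Fast" exit and a louder "Slow" exit.
# A modulator on each branch varies pitch and volume per play.
# The cue uses the leaf pile's attenuation and the traps sound class, limits concurrency per owner, and disables virtualization.
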SoundCue /BigLeafPile/Environmental/Audio/Hideable_BigLeafPile_Exit_Cue
root=Mixer_0
nodes [
 {id=Mixer_0, children [Modulator_1,Modulator_0], properties {inputVolume [0.85,1]}}
 {id=Modulator_0, children [Random_1], properties {pitchMax=1.25, volumeMax=1.7, volumeMin=1.2}}
 {id=Modulator_1, children [Random_0], properties {pitchMax=0.9, pitchMin=0.7, volumeMax=1, volumeMin=0.5}}
 {id=Random_0, children [WavePlayer_0,WavePlayer_1,WavePlayer_2], properties {weights [1,1,1]}}
 {id=Random_1, children [WavePlayer_6,WavePlayer_7,WavePlayer_8], properties {weights [1,1,1]}}
 {id=WavePlayer_0, properties {soundWaveAssetPtr=&/BigLeafPile_Exit_Fast_01}}
 {id=WavePlayer_1, properties {soundWaveAssetPtr=&/BigLeafPile_Exit_Fast_02}}
 {id=WavePlayer_2, properties {soundWaveAssetPtr=&/BigLeafPile_Exit_Fast_03}}
 {id=WavePlayer_6, properties {soundWaveAssetPtr=&/BigLeafPile_Exit_Slow_01}}
 {id=WavePlayer_7, properties {soundWaveAssetPtr=&/BigLeafPile_Exit_Slow_02}}
 {id=WavePlayer_8, properties {soundWaveAssetPtr=&/BigLeafPile_Exit_Slow_03}}
]
properties {
 attenuationSettings=&/Hideable_BigLeafPile_Att
 bOverrideConcurrency=true
 maxDistance=7800
 pitchMultiplier=0.8
 soundClassObject=&Fort_Traps
 virtualizationMode=Disabled
 volumeMultiplier=2.2
 concurrencyOverrides {bLimitToOwner=true, maxCount=2}
}
