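# Sound cue for the looping rustle of movement in the big leaf pile hideable prop.
# A busy movement loop wave feeds a Looping node, then a continuous modulator that drives volume from the MovementIntensity parameter, defaulting low.
# The cue lowers pitch and boosts volume of the busy loop, using the hiding-prop rustle attenuation and the traps sound class.
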SoundCue /BigLeafPile/Environmental/Audio/Hideable_BigLeafPile_LightMovement_Cue
root=ModulatorContinuous_0
nodes [
 {id=Looping_1, children [WavePlayer_0]}
 {
  id=ModulatorContinuous_0
  children [Looping_1]
  properties {volumeModulationParams {default=0.1, maxOutput=3, parameterName=MovementIntensity}}
 }
 {id=WavePlayer_0, properties {bLooping=true, soundWaveAssetPtr=&/BigLeafPile_Movement_Busy_Loop}}
]
properties {
 attenuationSettings=/FNGameplayCues/Sounds/Attenuation/Fort_HidingProps_Rustle
 maxDistance=3000
 pitchMultiplier=0.7
 soundClassObject=&Fort_Traps
 volumeMultiplier=2
}
