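; Big Leaf Pile environment actor that players can hide in, with rustling audio and quest tracking.
; On BeginPlay it destroys overlapping piles and registers collision-response swapping, or self-destroys if hotfix-disabled.
; The inner sphere applies and removes a gameplay effect and sends hide quest stats; vehicle and fire-death checks feed destroy quests.
; The outer sphere flags local pawns; client movement audio scales with speed and yaw change, with cooldown and fade-out.
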(event EventBeginPlay
  ;[0.2688 0.3094 0] Destroy other leaf piles if we're touching them ( for Leaf Pile Grenade )
  (CallFunction|StartAmbientAudioLoop)
  (if (Building|IsPreviewBuildingActor)
    _
    (elif (ScalableFloat|GetValueasBool (Variables|Default|GetFeatureEnabledHotfix))
      (bind _outactors (Collision|SphereOverlapActors (Transformation|GetWorldLocation (Variables|Default|GetOverlapSphere)) 1.4e+02 (Variables|Default|GetObject%20Types) "/BigLeafPile/Environmental/B_BigLeafPile.B_BigLeafPile_C" (Utilities|Array|MakeArray self)))
      (for _array_element _outactors
        (bind _asb_big_leaf_pile (Utilities|Casting|CastToB_BigLeafPile _array_element))
        (Actor|DestroyActor _asb_big_leaf_pile)
        (bind _outxy (DataTable|EvaluateCurveTableRow "/Game/Athena/Balance/DataTables/AthenaGameData.AthenaGameData" "Default.Cornfield.Wobble")
          (:RowFound
            (Variables|Default|SetShouldWobble (Math|Conversions|ToBoolean(Integer) (Math|Float|Truncate _outxy))))
          (:RowNotFound
            _))
        (if (Networking|IsServer)
          (Variables|/Script/Engine.PrimitiveComponent|SetbGenerateOverlapEvents true (Variables|Default|GetOverlapSphere))))
      (Rendering|SetCastShadow (Variables|Debug|GetStaticMeshComponent) true)
      (Class|FortCollisionResponseSwapperComponent|RegisterCallbacksforOverlap (Variables|Default|GetFortCollisionResponseSwapper) (Variables|Debug|GetStaticMeshComponent) (Variables|Default|GetOverlapSphere))
      (if (Networking|IsServer)
        ;[0.01768 1 0.08637] Quest died by Fire
        (bind (_asynctaskproxy _payload) (Ability|Async|WaitGameplayEventToActor 0 "(TagName=\"Event.Damage.Died\")" true)
          (:EventReceived
            (if (GameplayTags|HasTag (.InstigatorTags (Utilities|Struct|BreakGameplayEventData _payload)) (Variables|Default|GetQuest%20Fire%20Tag))
              (Quest|SendComplexCustomStatEvent (Quests|GetQuestManager (Fortnite|GetFortPlayerControllerfromActor (Actor|GetOwner (.Instigator (Utilities|Struct|BreakGameplayEventData _payload)))) "Invalid") (Fortnite|GetFortPlayerControllerfromActor (Actor|GetOwner (.Instigator (Utilities|Struct|BreakGameplayEventData _payload)))) (Variables|Default|GetTag%20Quest%20Destroy%20Leaf%20Pile%20with%20Vehicle%20or%20Fire) "(GameplayTags=)" 1)))))
      (else
        (Actor|DestroyActor)))))

(event Custom|Start%20Ambient%20Audio%20Loop
  ;[0.2688 0.3094 0] Start Leaf Pile Ambient Audio with randomized Offset
  (Audio|Components|Audio|Play (Variables|Default|GetAmbient%20Audio%20Loop) (Math|Random|RandomFloatinRange 0.0 5.0)))

(event EventEndPlay
  (params (EndPlayReason :type EEndPlayReason%20Enum :default "Destroyed"))
  ;[1 1 1] EVENT -> EndPlay
  (Class|FortCollisionResponseSwapperComponent|UnregisterCallbacksforOverlap (Variables|Default|GetFortCollisionResponseSwapper)))

(event OnComponentBeginOverlap(OverlapSphereOuterEdge)
  (params (OverlappedComponent :type Primitive%20Component%20Object%20Reference) (OtherActor :type Actor%20Object%20Reference) (OtherComp :type Primitive%20Component%20Object%20Reference) (OtherBodyIndex :type Integer) (bFromSweep :type Boolean) (SweepResult :type Hit%20Result%20Structure%20%28by%20ref%29))
  ;[0.3094 0.1987 0] Enter Leaf Pile SFX Trigger
  ;[0 0.3094 0.04008] If locally controlled, start a tick to watch for pawn movement in the pile. FORT-540884 : bots are locally controlled on server, make sure we're not server when we start this logic
  (bind _asfort_player_pawn_athena (Utilities|Casting|CastToFortPlayerPawnAthena OtherActor))
  (if (and (not (Networking|IsServer)) (Pawn|IsLocallyControlled _asfort_player_pawn_athena))
    (Variables|Audio|GetInteracting%20Actor
      (:then
        _)
      (:else
        (Variables|Audio|SetInteracting%20Actor OtherActor)))
    (Variables|Audio|SetInside%20Leaf%20Pile true)))

(event OnComponentBeginOverlap(OverlapSphere)
  (params (OverlappedComponent :type Primitive%20Component%20Object%20Reference) (OtherActor :type Actor%20Object%20Reference) (OtherComp :type Primitive%20Component%20Object%20Reference) (OtherBodyIndex :type Integer) (bFromSweep :type Boolean) (SweepResult :type Hit%20Result%20Structure%20%28by%20ref%29))
  (bind _asplayer_pawn_athena (Utilities|Casting|CastToPlayerPawn_Athena OtherActor)
    (:then
      (Utilities|IsValid (Ability|GetAbilitySystemComponent _asplayer_pawn_athena)
        (:Is%20Valid
          (GameplayEffects|ApplyGameplayEffectToTarget :self (Ability|GetAbilitySystemComponent _asplayer_pawn_athena) :GameplayEffectClass (Variables|Default|GetGameplay%20Effect%20Class) :Target (Ability|GetAbilitySystemComponent _asplayer_pawn_athena) :Level 1.0))
        (:Is%20Not%20Valid
          _))
      ;[0.01768 1 0.08637] Quests: Destroy w/vehicle & Hide in Leaf Pile Checks
      (Networking|SwitchHasAuthority
        (:Authority
          (CallFunction|QuestCheckVehcleDeath OtherActor)
          (Quest|SendComplexCustomStatEvent (Quests|GetQuestManager (Fortnite|GetFortPlayerControllerfromActor _asplayer_pawn_athena) "Invalid") (Fortnite|GetFortPlayerControllerfromActor _asplayer_pawn_athena) (Variables|Default|GetTag%20Quest%20Hide%20In%20Leaf%20Pile) "(GameplayTags=)" 1))))
    (:CastFailed
      _)))

(event OnComponentEndOverlap(OverlapSphere)
  (params (OverlappedComponent :type Primitive%20Component%20Object%20Reference) (OtherActor :type Actor%20Object%20Reference) (OtherComp :type Primitive%20Component%20Object%20Reference) (OtherBodyIndex :type Integer))
  (bind _asplayer_pawn_athena (Utilities|Casting|CastToPlayerPawn_Athena OtherActor))
  (GameplayEffects|RemoveActiveGameplayEffectbySourceEffect (Ability|GetAbilitySystemComponent _asplayer_pawn_athena) (Variables|Default|GetGameplay%20Effect%20Class) 0 1))

(event Custom|Replicated%20Leaf%20Enter%20Exit%20Event
  (params (Sound :type Sound%20Base%20Object%20Reference) (Other%20Actor :type Actor%20Object%20Reference))
  ;[1 1 1] Had to unify audio to work like other bush actors. Moved everything to work through libraries.
  ;;[0 0.3094 0.1843] Replicated Audio Trigger for Entry and Exit
  (if (not (Networking|IsServer))
    (Utilities|FlowControl|DoOnce false
      (:Completed
        (Audio|PlaySoundatLocation Sound (Transformation|GetActorLocation) "0, 0, 0" (Math|Float|MapRangeClamped (Math|Vector|VectorLength (Transformation|GetVelocity Other%20Actor)) 0.0 6e+02 0.0 1.0))
        (Utilities|FlowControl|Delay 1.0) ; prevent double trigger on local client
      ))))

(event Custom|Local%20Client%20Movement%20Check
  (bind _interacting_actor (Variables|Audio|GetInteracting%20Actor))
  (bind (_returnvalue_roll _returnvalue_pitch _returnvalue_yaw) (Math|Rotator|BreakRotator (Transformation|GetActorRotation _interacting_actor)))
  (bind _previous_yaw_value (Variables|Audio|GetPrevious%20Yaw%20Value))
  (bind _returnvalue (- _previous_yaw_value _returnvalue_yaw))
  (bind _returnvalue_2 (Math|Float|Absolute(Float) _returnvalue))
  (bind _interacting_actor_2 (Variables|Audio|GetInteracting%20Actor))
  (bind _returnvalue_3 (Transformation|GetVelocity _interacting_actor_2))
  (bind _returnvalue_4 (Math|Vector|VectorLength _returnvalue_3))
  ;[1 1 1] Had to unify audio to work like other bush actors. Moved everything to work through libraries.
  ;;[0 0.03429 0.3094] If Actor is locally controlled we calculate a cheap version of Rotational Velocity based on Yaw alone. This triggers leaf SFX if above a threshold. We only allow 1 event per "Movement Cooldown" interval.
  (Variables|Audio|SetPretend%20Rotational%20Velocity (* _returnvalue_2 1e+01))
  (Variables|Audio|SetActor%20Velocity _returnvalue_4)
  (if (or (> _returnvalue_4 5e+01) (> _returnvalue_2 1e+01))
    (Utilities|FlowControl|Sequence
      (:then_0
        (if (not (Variables|Audio|GetMovement%20Cooldown))
          (CallFunction|LeafMovementEvent (/ (+ (Variables|Audio|GetPretend%20Rotational%20Velocity) (Variables|Audio|GetActor%20Velocity)) 2.0))
          (Utilities|FlowControl|Sequence
            (:then_0
              (Variables|Audio|SetMovement%20Cooldown true)
              (Utilities|FlowControl|Delay 0.5)
              (Variables|Audio|SetMovement%20Cooldown false))
            (:then_1
              (bind (_direction _intensity_multiplier) (Timeline|RetriggerFadeOut PlayFromStart)
                (:Update
                  (Audio|Parameter|SetFloatParameter (Variables|Default|GetAudio%20Component) "MovementIntensity" _intensity_multiplier :!class /Script/Engine.AudioComponent)
                  (Audio|Components|Audio|SetVolumeMultiplier (Variables|Default|GetAudio%20Component) _intensity_multiplier :!class /Script/Engine.AudioComponent))
                (:Finished
                  (Audio|Components|Audio|Stop (Variables|Default|GetAudio%20Component))))))))
      (:then_1
        _)))
  (Variables|Audio|SetPrevious%20Yaw%20Value _returnvalue_yaw))

(event Custom|Leaf%20Movement%20Event
  (params (Velocity :type Float%20%28double-precision%29))
  ;[1 1 1] Had to unify audio to work like other bush actors. Moved everything to work through libraries.
  ;;[0.1455 0.3094 0] Non-replicated trigger for pawn movement inside the leaf pile - replicate in future?
  (if (not (Networking|IsServer))
    (CallFunction|PlayInsideLeafPileMovementAudio Velocity)))

(event OnComponentEndOverlap(OverlapSphereOuterEdge)
  (params (OverlappedComponent :type Primitive%20Component%20Object%20Reference) (OtherActor :type Actor%20Object%20Reference) (OtherComp :type Primitive%20Component%20Object%20Reference) (OtherBodyIndex :type Integer)))

(fn PlayInsideLeafPileMovementAudio
  (params (Velocity :type Float%20%28double-precision%29))
  (Variables|Default|GetAudio%20Component
    (:then
      (Audio|Components|Audio|SetVolumeMultiplier (Variables|Default|GetAudio%20Component) (Math|Float|MapRangeClamped (Variables|GetVelocity) 0.0 3e+02 0.5 1.0) :!class /Script/Engine.AudioComponent)
      (if (not (Audio|Components|Audio|IsPlaying (Variables|Default|GetAudio%20Component)))
        (Audio|Components|Audio|FadeIn (Variables|Default|GetAudio%20Component) 0.2 1.0 (Math|Conversions|ToFloat(Integer) (+ (Math|Random|RandomInteger 7) 1)) "Linear" :!class /Script/Engine.AudioComponent))
      (Audio|Parameter|SetFloatParameter (Variables|Default|GetAudio%20Component) "MovementIntensity" (Math|Float|MapRangeClamped (Variables|GetVelocity) 0.0 3e+02 0.0 1.0) :!class /Script/Engine.AudioComponent)
      (Development|PrintString (Utilities|String|Append "Leaf Pile Audio Velocity: " (Utilities|String|ToString(Float) (Variables|GetVelocity))) (Variables|Audio|GetAudio%20Debug) false))
    (:else
      (Development|PrintString "Big Leaf Pile Audio Component Not Valid"))))

(fn QuestCheckVehcleDeath
  (params (Object :type Object%20Reference))
  (bind _asfort_vehicle_interface (Utilities|Casting|CastToFortVehicleInterface Object))
  (Utilities|IsValid (Vehicle|GetDriver _asfort_vehicle_interface)
    (:Is%20Valid
      (Quest|SendComplexCustomStatEvent (Quests|GetQuestManager (Fortnite|GetFortPlayerControllerfromActor (Vehicle|GetDriver _asfort_vehicle_interface)) "Invalid") (Fortnite|GetFortPlayerControllerfromActor (Vehicle|GetDriver _asfort_vehicle_interface)) (Variables|Default|GetTag%20Quest%20Destroy%20Leaf%20Pile%20with%20Vehicle%20or%20Fire) "(GameplayTags=)" 1)
      (return))))

(fn ConstructionScript)
